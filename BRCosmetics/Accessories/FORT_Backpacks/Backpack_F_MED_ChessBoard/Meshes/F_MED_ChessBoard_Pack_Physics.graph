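# Physics asset for the ChessBoard backpack cosmetic's skeletal mesh, with collision bodies for the pack and its two dynamic straps.
# A large capsule covers the pack bone, and a small capsule sits on each left and right strap bone.
# Each strap is constrained to the pack with the same limited swing and twist ranges, forming a simple two-strap rig around one central body.
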PhysicsAsset /BRCosmetics/Accessories/FORT_Backpacks/Backpack_F_MED_ChessBoard/Meshes/F_MED_ChessBoard_Pack_Physics
bodies {
 dyn_strap_l [{shapeType=Capsule, length=1.5, radius=1.5, center=<1 0 0>, rotation=<rot 90 0 0>}]
 dyn_strap_r [{shapeType=Capsule, length=1.5, radius=1.5, center=<1 0 0>, rotation=<rot 90 0 0>}]
 pack [{shapeType=Capsule, length=8, radius=8, center=<0 -12.1941 5.0207>}]
}
constraints [
 {bone1Name=dyn_strap_l, bone2Name=pack, swing1LimitDegrees=20, swing2LimitDegrees=45, twistLimitDegrees=15}
 {bone1Name=dyn_strap_r, bone2Name=pack, swing1LimitDegrees=20, swing2LimitDegrees=45, twistLimitDegrees=15}
]
