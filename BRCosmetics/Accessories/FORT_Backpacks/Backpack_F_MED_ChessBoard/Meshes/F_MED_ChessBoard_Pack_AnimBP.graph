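# Anim Blueprint for the F_MED ChessBoard backpack, adding an idle floating bob and physics-driven motion.
# The update event builds time-based sine translation and rotation, which the AnimGraph adds to the pack bone amid fixed pack and root offsets.
# A RigidBody node then simulates in root-bone space with zero gravity.
# Gender/body flags, spine-derived offsets, computed gravity and jump/DBNO states are set but unused by the AnimGraph.
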AnimBlueprint /BRCosmetics/Accessories/FORT_Backpacks/Backpack_F_MED_ChessBoard/Meshes/F_MED_ChessBoard_Pack_AnimBP
target_skeleton=&/F_MED_ChessBoard_Pack_Skeleton
graphs [
 AnimGraph {nodes [
   n1 = "Transform (Modify) Bone - Bone: pack"!(ComponentPose=Local To Component(LocalPose=Input Pose()), bAlphaBoolEnabled:bool, Translation="0,0.5,-2") -> (Pose) {
    alphaInputType=Bool
    translationMode=BMM_Additive
    alphaBoolBlend {blendOption=Linear}
    boneToModify {boneName=pack}
    translation {y=0.5, z=-2}
   }
   n2 = "Transform (Modify) Bone - Bone: pack"(ComponentPose=@n1, Translation="0,-2,0") {
    alphaInputType=Bool
    translationMode=BMM_Additive
    translationSpace=BCS_BoneSpace
    alphaBoolBlend {blendOption=Linear}
    boneToModify {boneName=pack}
    translation {y=-2}
   }
   n3 = "Transform (Modify) Bone - Bone: pack"(ComponentPose=@n2, Translation="0,-1.5,0") {
    alphaInputType=Bool
    translationMode=BMM_Additive
    translationSpace=BCS_BoneSpace
    alphaBoolBlend {blendOption=Linear}
    boneToModify {boneName=pack}
    translation {y=-1.5}
   }
   n4 = "Transform (Modify) Bone - Bone: pack"!(ComponentPose=@n3, Alpha:float, Translation="0,0,0", Rotation:Rotator="0,0,0") -> (Pose) {rotationMode=BMM_Additive, translationMode=BMM_Additive, alphaBoolBlend {blendOption=Linear}, boneToModify {boneName=pack}}
   n5 = "Transform (Modify) Bone - Bone: pack"(
    ComponentPose=@n4
    Translation=Get Translation floating() -> (Translation floating:Vector)
    Rotation=Get rotation floating() -> (rotation floating:Rotator)
   ) {
    rotationMode=BMM_Additive
    rotationSpace=BCS_BoneSpace
    translationMode=BMM_Additive
    translationSpace=BCS_BoneSpace
    alphaBoolBlend {blendOption=Linear}
    boneToModify {boneName=pack}
   }
   n6 = "Transform (Modify) Bone - Bone: root"!(ComponentPose=@n5, bAlphaBoolEnabled:bool=True, Translation="0,-8,0", Rotation="0,0,-90") -> (Pose) {
    alphaInputType=Bool
    rotationMode=BMM_Additive
    rotationSpace=BCS_BoneSpace
    translationMode=BMM_Additive
    translationSpace=BCS_BoneSpace
    alphaBoolBlend {blendOption=Linear}
    boneToModify {boneName=root}
    rotation {roll=-90}
    translation {y=-8}
   }
   n7 = RigidBody!(ComponentPose=@n6, Alpha, OverrideWorldGravity:Vector="0,0,0") -> (Pose) {
    bOverrideWorldGravity=true
    overridePhysicsAsset=&/F_MED_ChessBoard_Pack_Physics
    simulationSpace=BaseBoneSpace
    alphaBoolBlend {blendOption=Linear}
    baseBoneRef {boneName=root}
    componentAppliedLinearAccClamp=<3000 3000 3000>
    componentLinearAccScale=<0.35 0.35 0.4>
    componentLinearVelScale=<0.35 0.35 0.4>
    simSpaceSettings {
     maxAngularAcceleration=1000
     maxAngularVelocity=3
     maxLinearAcceleration=700
     maxLinearVelocity=500
     velocityScaleZ=0
     worldAlpha=0.4
     externalLinearDragV=<0.2 0.2>
    }
   }
   Output Pose(Result=Component To Local(ComponentPose=@n7)) {layerGroup=None}
  ]}
 EventGraph {nodes [
   Event OnCharacterCustomizationLoaded()
   > n1 = Set Owning Pawn(Owning Pawn:Pawn Object Reference=TryGetFortPawn()) -> (Output_Get:Pawn Object Reference)
   > Is Valid(InputObject=@n1)
   > n2 = Sequence()
   > n3 = FindGender!(execute, Object:Object Reference=Get Owning Pawn() -> (Owning Pawn:Pawn Object Reference)) -> (True:Exec, False:Exec, OutGender:EFortCustomGender Enum=Invalid)
   > Set Male(Male:bool="Equal (Enum)"(A:EFortCustomGender Enum=@n3, B:EFortCustomGender Enum=Male)) -> (Output_Get:bool)
   > Set Female(Female:bool="Equal (Enum)"(A=@n3, B=Female)) -> (Output_Get:bool)
   n4 = FindBodyType!(execute=@n2.then_1, Object:Object Reference=Get Owning Pawn()) -> (True:Exec, False:Exec, OutBodyType:EFortCustomBodyType Enum=NONE)
   > Set Large(Large:bool="Equal (Enum)"(A:EFortCustomBodyType Enum=@n4, B:EFortCustomBodyType Enum=Large)) -> (Output_Get:bool)
   > Set Medium(Medium:bool="Equal (Enum)"(A=@n4, B=Medium)) -> (Output_Get:bool)
   Event BlueprintUpdateAnimation()
   > n5 = Sequence!(execute) -> (then_0, then_1, then_2:Exec)
   > n6 = Gravity Override State Function DEPRECATED(
    CharacterPartAnimInstance=Self-Reference()
    GameTimeInSeconds=Get Game Time in Seconds() -> (Game Time in Seconds:double)
    GravityOverrideParams=Get Gravity Override Params()
   )
   > Set Gravity Override(Gravity Override=@n6)
   > Set F_MED(F_MED:bool=AND Boolean(A=Get Female() -> (Female:bool), B=Get Medium() -> (Medium:bool))) -> (Output_Get:bool)
   > Set M_MED(M_MED:bool=AND Boolean(A=Get Male() -> (Male:bool), B=Get Medium())) -> (Output_Get:bool)
   n7 = GetOwnerSkeletalMeshComponent()
   n8 = BreakRotator(InRot=BreakTransform(InTransform=GetDeltaTransformFromRefPose(self=@n7, BoneName=spine_05, BaseName=spine_01)).Rotation)
   > Set Backpack Rotate(Backpack Rotate:Rotator=MakeRotator(Roll=MapRangeClamped(Value=@n8.Yaw, InRangeA=0, InRangeB=-30, OutRangeB=40))) -> (Output_Get:Rotator)
   > Set Backpack Translation(Backpack Translation:Vector=MakeVector(Y=MapRangeClamped(Value=@n8.Yaw, InRangeA=0, InRangeB=-30, OutRangeB=-8))) -> (Output_Get:Vector)
   n9 = Get bIsSkyDiving()
   > n10 = Branch(Condition=AND Boolean(A=Get bIsJumping(), B=NOT Boolean(A=@n9)))
   > Set deltaonOff(deltaonOff:float=0) -> (Output_Get:float)
   Set deltaonOff(execute=@n10.else, deltaonOff=1)
   Set DBNO-Carried(execute=@n5.then_2, DBNO-Carried:bool=OR Boolean(A=Get bIsDBNO(), B=Get bIsDBNOCarried())) -> (Output_Get:bool)
   n11 = GetRealTimeSeconds()
   n12 = "Sin (Radians)"(A=float * float(A=float * float(A=1.2, B=1.3), B=@n11))
   Set Translation floating(execute=@n5.then_1, Translation floating:Vector=MakeVector(X=float * float(A=@n12, B=0.2), Y=0, Z=float * float(A=@n12, B=1.5))) -> (Output_Get:Vector)
   n13 = "Sin (Radians)"(A=float * float(A=float * float(A=1.1, B=1.1), B=@n11))
   n14 = "Sin (Radians)"(A=float * float(A=float * float(A=0.9, B=1.1), B=@n11))
   > Set rotation floating(rotation floating:Rotator=MakeRotator(Roll=float * float(A=@n12, B=2.5), Pitch=float * float(A=@n13, B=7), Yaw=float * float(A=@n14, B=0.5))) -> (Output_Get:Rotator)
   TryGetPawnOwner()
   K2Node_Event_0 = CustomEvent!() -> (OutputDelegate, then, DeltaTimeX:float=~) {is=title K2Node_CustomEvent}
  ]}
]
variables [
 Backpack Rotate {type=Rotator}
 Backpack Translation {type=Vector}
 DBNO-Carried {type=boolean}
 deltaonOff {type=number}
 F_MED {type=boolean}
 Female {type=boolean}
 Game Time in Seconds {type=number}
 Gravity Override {type=Vector, default=<0 0 -980>}
 Gravity Override Params {type=GravityOverrideParamsStruct}
 Large {type=boolean}
 M_MED {type=boolean}
 Male {type=boolean}
 Medium {type=boolean}
 Owning Pawn {type=Pawn, default=null}
 rotation floating {type=Rotator}
 Translation floating {type=Vector}
]
